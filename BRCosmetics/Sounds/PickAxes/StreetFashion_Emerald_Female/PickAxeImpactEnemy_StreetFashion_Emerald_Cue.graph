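# Sound cue for the StreetFashion Emerald Female pickaxe hitting an enemy, using bat impact waves.
# A quality-level root picks one branch: six random waves at the first level, three at the second, and one shared by the last two.
# Each branch adds a short delay, randomized only in the six-wave branch, and a default Modulator for variation.
# Pickaxe impact attenuation, melee weapon concurrency and prime-on-load apply.
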SoundCue /BRCosmetics/Sounds/PickAxes/StreetFashion_Emerald_Female/PickAxeImpactEnemy_StreetFashion_Emerald_Cue
root=QualityLevel_0
nodes [
 {id=Delay_0, children [Modulator_3], properties {delayMax=0.1, delayMin=0.1}}
 {id=Delay_1, children [Modulator_2], properties {delayMax=0.1, delayMin=0.05}}
 {id=Delay_2, children [Modulator_5], properties {delayMax=0.1, delayMin=0.1}}
 {id=Modulator_2, children [Random_0]}
 {id=Modulator_3, children [Random_1]}
 {id=Modulator_5, children [WavePlayer_9]}
 {id=QualityLevel_0, children [Delay_1,Delay_0,Delay_2,Delay_2]}
 {id=Random_0, children [WavePlayer_0,WavePlayer_1,WavePlayer_2,WavePlayer_3,WavePlayer_4,WavePlayer_5], properties {weights [1,1,1,1,1,1]}}
 {id=Random_1, children [WavePlayer_6,WavePlayer_7,WavePlayer_8], properties {weights [1,1,1]}}
 {
  id=WavePlayer_0
  properties {soundWaveAssetPtr=&/PickAxe_StreetFashion_Bat_Enemy_Impact_01}
 }
 {
  id=WavePlayer_1
  properties {soundWaveAssetPtr=&/PickAxe_StreetFashion_Bat_Enemy_Impact_02}
 }
 {
  id=WavePlayer_2
  properties {soundWaveAssetPtr=&/PickAxe_StreetFashion_Bat_Enemy_Impact_03}
 }
 {
  id=WavePlayer_3
  properties {soundWaveAssetPtr=&/PickAxe_StreetFashion_Bat_Enemy_Impact_04}
 }
 {
  id=WavePlayer_4
  properties {soundWaveAssetPtr=&/PickAxe_StreetFashion_Bat_Enemy_Impact_05}
 }
 {
  id=WavePlayer_5
  properties {soundWaveAssetPtr=&/PickAxe_StreetFashion_Bat_Enemy_Impact_06}
 }
 {
  id=WavePlayer_6
  properties {soundWaveAssetPtr=&PickAxe_StreetFashion_Bat_Enemy_Impact_01}
 }
 {
  id=WavePlayer_7
  properties {soundWaveAssetPtr=&PickAxe_StreetFashion_Bat_Enemy_Impact_03}
 }
 {
  id=WavePlayer_8
  properties {soundWaveAssetPtr=&PickAxe_StreetFashion_Bat_Enemy_Impact_02}
 }
 {
  id=WavePlayer_9
  properties {soundWaveAssetPtr=&PickAxe_StreetFashion_Bat_Enemy_Impact_01}
 }
]
properties {
 attenuationSettings=/Game/Sounds/Attenuation/Fort_Pickaxe_Impact_Att
 audioPropertiesSheet=/BRCosmetics/Sounds/Settings/PropertySheets/Pickaxe/APS_Cosmetic_Pickaxe_Impact
 bPrimeOnLoad=true
 maxDistance=4300
 soundClassObject=&Fort_Impacts
 volumeMultiplier=0.825
 concurrencySet [&Player_Melee_Weapon_Conc]
}
